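# Material function for character visors that overrides base color and emissive on incoming material attributes and also outputs UVs.
# A UV2 or channel mask confines the effect, with bump-offset depth and fresnel tint.
# Static switches select the emissive source, including flipbook or two-layer animated distance-field SubUV faces, optionally voice-amplitude reactive through an audio parameter collection.
# Optional glitch, sweep and TV static distort UVs; scanlines, sweep, static and pulse modulate emissive.
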MaterialFunction /Game/Global/Materials/Character_Parent_2023/Visor_Parent_2025/MF_GenericVisor
graph {nodes [
  Constant() {r=20}
  Constant() {r=-0.3}
  Constant() {r=60}
  Constant() {r=1}
  n1 = LocalPosition()
  Divide(
   A=ComponentMask(None=VertexInterpolator(VS=@n1.XYZ).PS) {b=true, r=true}
   B=ScalarParameter() {defaultValue=128, group=Visor Add-Ons, parameterName=Pulse_Scale, sortPriority=55}
  )
  n2 = GetMaterialAttributes(None=FunctionInput() {inputName=Material, inputType=FunctionInput_MaterialAttributes}) {attributeGetTypes [#1, #2, #3]}
  n3 = ScalarParameter() {group=Visor, parameterName=Visor_BasecolorScale, sortPriority=43}
  n4 = SetMaterialAttributes!(
   MaterialAttributes=@n2.MaterialAttributes
   EmissiveColor=NamedRerouteUsage()
   BaseColor=LinearInterpolate(A=@n2.BaseColor, B=Multiply(A=@n2.BaseColor, B=@n3), Alpha=NamedRerouteUsage()) {constB=0}
  ) {attributeSetTypes [#2, #3]}
  FunctionOutput(None=@n4)
  FunctionOutput(None=StaticSwitchParameter(True=NamedRerouteUsage(), False=TextureCoordinate()) {group=Visor, parameterName=Visor_OverrideBasicEmissiveUVs, sortPriority=21}) {outputName=UVs, sortPriority=1}
  n5 = LinearInterpolate(
   B=MaterialFunctionCall!(ScanLine_Tiling_1, Scanline_PannerSpeed_1, ScanLine_Tiling, Scanline_PannerSpeed, UVs=NamedRerouteUsage()) {materialFunction=/Game/Weapons/WeaponSkins/MaterialFunction/ScanLines}.Result
   Alpha=ScalarParameter() {defaultValue=0.5, group=Scanlines, parameterName=Scanline_Opacity, sortPriority=50}
  ) {constA=1}
  Multiply(A=@n5, B=NamedRerouteUsage())
  n6 = NamedRerouteUsage()
  n7 = NamedRerouteUsage()
  n8 = NamedRerouteUsage()
  n9 = NamedRerouteUsage()
  NamedRerouteDeclaration(None=StaticSwitchParameter(
    True=Multiply(A=@n6, B=@n7)
    False=StaticSwitchParameter(True=@n8, False=@n9) {group=Visor Base, parameterName=Visor_UseAnimatedDFVisor, sortPriority=20}
   ) {group=Visor Base, parameterName=Visor_UseFlipbookFace, sortPriority=20}) {name=Visor_Base, nodeColor=<rgba 0 0.0161 1 1>}
  n10 = NamedRerouteUsage()
  n11 = TextureSampleParameter2D(UVs=@n10) {
   group=06_Emissive
   parameterName=Emissive
   samplerSource=SSM_Clamp_WorldGroupSettings
   sortPriority=1
   texture=&T_BlacksRGB
  }
  NamedRerouteDeclaration(None=StaticSwitchParameter(
    True=TextureSampleParameter2D(UVs=@n10) {group=Visor, parameterName=Visor_Emissive, texture=&T_BlacksRGB}.RGB
    False=@n11.RGB
   ) {group=Visor Base, parameterName=Visor_UseUV2, sortPriority=20}) {name=StaticImage, nodeColor=<rgba 0.6172 0 1 1>}
  NamedRerouteDeclaration(None=@n2.Normal) {name=Normal_IN, nodeColor=<rgba 1 0 0.7266 1>}
  n12 = Power(
   Base=OneMinus(None=Fresnel() {baseReflectFraction=0, exponent=1})
   Exp=ScalarParameter() {defaultValue=3, group=Visor, parameterName=Visor_FresnelPower, sortPriority=43}
  )
  NamedRerouteDeclaration(None=LinearInterpolate(B=Saturate(None=@n12), Alpha=ScalarParameter() {defaultValue=1, group=Visor, parameterName=Visor_FresnelIntensity, sortPriority=43}) {constA=1}) {name=Fresnel, nodeColor=<rgba 1 0.8203 0 1>}
  n13 = StaticSwitchParameter(True=TextureCoordinate() {coordinateIndex=1}, False=TextureCoordinate()) {group=Visor, parameterName=Visor_UseUV2, sortPriority=20}
  n14 = VectorParameter() {group=Visor, parameterName=Visor_UVScale, sortPriority=40, defaultValue=<rgba 1 1 0 1>}
  n15 = Add(
   A=Multiply(A=NamedRerouteDeclaration(None=@n13) {name=UV12, nodeColor=<rgba 0 0.0391 1 1>}, B=ComponentMask(None=@n14.RGB) {g=true, r=true})
   B=ComponentMask(None=VectorParameter() {group=Visor, parameterName=Visor_UVOffset, sortPriority=40, defaultValue=<rgba 0 0 0 1>}.RGB) {g=true, r=true}
  )
  n16 = ScalarParameter() {group=Visor, parameterName=Visor_Depth, sortPriority=41}
  n17 = Constant()
  n18 = NamedRerouteUsage()
  NamedRerouteDeclaration(None=LinearInterpolate(
    A=NamedRerouteUsage()
    B=MaterialFunctionCall!(UV Coord=@n15, Height=@n16, ReferencePlane=@n17, HeightRatio, Normal=@n18) {materialFunction=/Game/Global/MaterialFunctions/MF_BumpOffset_Accurate}.Result
    Alpha=NamedRerouteUsage()
   )) {name=UVs, nodeColor=<rgba 0.875 1 0 1>}
  NamedRerouteDeclaration(None=@n2.EmissiveColor) {name=Emissive_IN, nodeColor=<rgba 1 0 0.9609 1>}
  n19 = VectorParameter() {group=Visor, parameterName=Visor_FresnelTintColor, sortPriority=43, defaultValue=<rgba 0 0 0 1>}
  n20 = Multiply(A=NamedRerouteUsage(), B=ScalarParameter() {defaultValue=1, group=Visor, parameterName=Visor_BaseBrightness, sortPriority=41})
  n21 = StaticSwitchParameter(True=Multiply(A=@n20, B=NamedRerouteUsage()), False=@n20) {group=Visor Add-Ons, parameterName=Visor_UseScanline, sortPriority=20}
  n22 = StaticSwitchParameter(True=Add(A=@n21, B=NamedRerouteUsage()), False=@n21) {group=Visor Add-Ons, parameterName=Visor_UseSweepDistortion, sortPriority=49}
  n23 = StaticSwitchParameter(True=Add(A=@n22, B=NamedRerouteUsage()), False=@n22) {group=Visor Add-Ons, parameterName=Visor_UseTVStatic, sortPriority=44}
  n24 = StaticSwitchParameter(True=Multiply(A=@n23, B=NamedRerouteUsage()), False=@n23) {group=Visor Add-Ons, parameterName=Visor_UsePulse, sortPriority=54}
  n25 = NamedRerouteUsage()
  NamedRerouteDeclaration(None=LinearInterpolate(A=NamedRerouteUsage(), B=LinearInterpolate(A=@n19.RGB, B=@n24, Alpha=@n25), Alpha=NamedRerouteUsage())) {name=Emissive_OUT, nodeColor=<rgba 0 1 0.7188 1>}
  NamedRerouteDeclaration() {name=BasicVisor, nodeColor=<rgba 0.8984 1 0 1>}
  NamedRerouteDeclaration(None=@n5) {name=Scanline, nodeColor=<rgba 1 0.2109 0 1>}
  n26 = Constant2Vector() {g=0.1, r=5}
  n27 = NamedRerouteDeclaration(None=ScalarParameter() {defaultValue=1, group=Visor Add-Ons, parameterName=TVStatic_Speed, sortPriority=45}) {name=StaticSpeed, nodeColor=<rgba 0 1 0.6484 1>}
  n28 = Multiply(A=NamedRerouteUsage(), B=ScalarParameter() {defaultValue=3.5, group=Visor Add-Ons, parameterName=TVStatic_UVScale, sortPriority=45})
  n29 = TextureSample(UVs=Panner(Coordinate=@n28, Speed=Multiply(A=@n26, B=@n27)) {speedX=5, speedY=0.1}) {
   samplerSource=SSM_Wrap_WorldGroupSettings
   samplerType=SAMPLERTYPE_Masks
   texture=/Game/Missions/Primary/RepairTheShelter/Textures/Cinematic_Screen_Static
  }
  n30 = Multiply(A=@n29.B) {constB=0.1}
  n31 = Constant2Vector() {g=-0.05, r=6}
  n32 = TextureSample(UVs=Panner(Coordinate=@n28, Speed=Multiply(A=@n31, B=@n27)) {speedX=6, speedY=-0.05}) {
   samplerSource=SSM_Wrap_WorldGroupSettings
   samplerType=SAMPLERTYPE_Masks
   texture=&Cinematic_Screen_Static
  }
  n33 = Constant2Vector() {g=-0.03, r=-3}
  n34 = TextureSample(UVs=Panner(Coordinate=@n28, Speed=Multiply(A=@n33, B=@n27)) {speedX=-3, speedY=-0.03}) {
   samplerSource=SSM_Wrap_WorldGroupSettings
   samplerType=SAMPLERTYPE_Masks
   texture=&Cinematic_Screen_Static
  }
  n35 = Multiply(A=@n32.G, B=@n34.R)
  n36 = Saturate(None=Multiply(A=Add(A=@n30, B=@n35), B=ScalarParameter() {defaultValue=5, group=Visor Add-Ons, parameterName=TVStatic_Brightness, sortPriority=45}))
  n37 = VectorParameter() {group=Visor Add-Ons, parameterName=TVStatic_Color, sortPriority=45, defaultValue=<rgba 1 1 1 1>}
  NamedRerouteDeclaration(None=Multiply(A=Multiply(A=@n36, B=@n37.RGB), B=NamedRerouteUsage())) {name=Static, nodeColor=<rgba 0 1 0.8828 1>}
  n38 = ScalarParameter() {defaultValue=0.07, group=Visor Add-Ons, parameterName=TVStatic_UVDistort, sortPriority=45}
  NamedRerouteDeclaration(None=AppendVector(A=Multiply(A=@n35, B=@n38), B=Constant())) {name=StaticUVDistortion, nodeColor=<rgba 1 0.5625 0 1>}
  n39 = VectorParameter() {group=Visor Add-Ons, parameterName=SweepDistortion_Color, sortPriority=50, defaultValue=<rgba 1 1 1 1>}
  n40 = NamedRerouteUsage()
  n41 = ScalarParameter() {defaultValue=4, group=Visor Add-Ons, parameterName=SweepDistortion_UVScale, sortPriority=50}
  n42 = Multiply(A=ComponentMask(None=Multiply(A=@n40, B=@n41)) {g=true}) {constB=0.75}
  n43 = Multiply(A=Time(), B=ScalarParameter() {defaultValue=-0.2, group=Visor Add-Ons, parameterName=SweepDistortion_Speed, sortPriority=50}) {constB=-0.2}
  n44 = Power(Base=Frac(None=Add(A=@n42, B=@n43))) {constExponent=20}
  n45 = Multiply(A=@n44, B=ScalarParameter() {defaultValue=0.1, group=Visor Add-Ons, parameterName=SweepDistortion_Brightness, sortPriority=50}) {constB=0.2}
  NamedRerouteDeclaration(None=Multiply(A=Multiply(A=@n39.RGB, B=@n45), B=NamedRerouteUsage())) {name=Sweep, nodeColor=<rgba 0.2422 0 1 1>}
  n46 = ScalarParameter() {defaultValue=0.01, group=Visor Add-Ons, parameterName=SweepDistortion_Amount, sortPriority=50}
  NamedRerouteDeclaration(None=AppendVector(A=Multiply(A=@n44, B=@n46) {constB=0.02}, B=Constant())) {name=SweepUVDistortion, nodeColor=<rgba 0.0781 1 0 1>}
  n47 = Time()
  n48 = CollectionParameter() {collection=/Game/Athena/Events/AudioAnalysis/MPC_Audio_Reactive_Parameters, parameterName=Amplitude_VO_B}
  n49 = Multiply(A=@n48) {constB=1.8}
  n50 = Ceil(None=@n48)
  NamedRerouteDeclaration(None=StaticSwitchParameter(True=LinearInterpolate(A=@n47, B=@n49, Alpha=@n50), False=@n47) {group=SmartTuna, parameterName=UseEventVO}) {name=EventVOTime, nodeColor=<rgba 0 0.5312 1 1>}
  NamedRerouteDeclaration() {name=VisorTex, nodeColor=<rgba 0 1 0.0391 1>}
  n51 = Power(Base=Multiply(A=CollectionParameter() {collection=&MPC_Audio_Reactive_Parameters, parameterName=Amplitude_VO_B}) {constB=3}) {constExponent=1.112}
  n52 = Saturate(None=Sine(None=@n51))
  n53 = TextureObjectParameter() {
   group=Visor Base
   parameterName=AnimatedDF_EmissiveDistanceField
   samplerType=SAMPLERTYPE_LinearColor
   sortPriority=40
   texture=/Game/Global/Materials/MaterialGym/Materials/Visor/Examples/Binary/Textures/T_Binary_Visor_DF
  }
  n54 = NamedRerouteUsage()
  n55 = Constant() {r=2}
  n56 = ScalarParameter() {defaultValue=-2, group=Visor Base, parameterName=AnimatedDF_FlipbookTexLODBias}
  n57 = RecordTextureStreamingInfo!(Tex=@n53, UVs=Multiply(A=@n54, B=Power(Base=@n55, Exp=@n56)))
  n58 = ScalarParameter() {defaultValue=2, group=Visor Base, parameterName=AnimatedVisor_SubUV_Frames, sortPriority=60}
  n59 = Multiply(A=NamedRerouteUsage(), B=ScalarParameter() {defaultValue=2, group=Visor Base, parameterName=AnimatedDF_SubUV_Speed, sortPriority=60})
  n60 = MaterialFunctionCall!(Texture=@n57.Tex, UVs=@n54, SubImages=@n58, Frame=@n59) {materialFunction=/Engine/Functions/Engine_MaterialFunctions03/Texturing/SubUV_Function}
  n61 = ScalarParameter() {defaultValue=-0.6, group=Visor Base, parameterName=AnimatedDF_LineThickness, sortPriority=60}
  n62 = Saturate(None=Add(
    A=MaterialFunctionCall(Float3=@n60.RGB) {materialFunction=&BreakOutFloat3Components}.R
    B=@n61
   ) {constB=-0.48})
  n63 = Constant() {r=16}
  n64 = Multiply(
   A=Saturate(None=Multiply(A=@n62, B=@n63))
   B=ScalarParameter() {defaultValue=2, group=Visor Base, parameterName=AnimatedDF_Layer1_Strength, sortPriority=60}
  )
  n65 = StaticSwitchParameter(True=Multiply(A=@n64, B=LinearInterpolate(Alpha=@n52) {constA=0.8, constB=6}), False=@n64) {group=SmartTuna, parameterName=UseEventVO}
  n66 = Saturate(None=@n51)
  n67 = MaterialFunctionCall(
   Texture=@n57.Tex
   UVs=BumpOffset(
    Coordinate=@n54
    Height=ScalarParameter() {group=Visor Base, parameterName=AnimatedDF_Height, sortPriority=60}
    HeightRatioInput=ScalarParameter() {defaultValue=0.015, group=Visor Base, parameterName=AnimatedDF_Layer2_HeightRatio, sortPriority=60}
   ) {referencePlane=1}
   SubImages=@n58
   Frame=@n59
  ) {materialFunction=&SubUV_Function}
  n68 = Saturate(None=Add(
    A=MaterialFunctionCall(Float3=@n67.RGB) {materialFunction=&BreakOutFloat3Components}.R
    B=@n61
   ) {constB=-0.48})
  n69 = Constant() {r=16}
  n70 = Multiply(
   A=Saturate(None=Multiply(A=@n68, B=@n69))
   B=ScalarParameter() {defaultValue=0.5, group=Visor Base, parameterName=AnimatedDF_Layer2_Strength, sortPriority=60}
  )
  n71 = StaticSwitchParameter(True=Multiply(A=@n70, B=LinearInterpolate(Alpha=@n66) {constA=0.75, constB=8}), False=@n70) {group=SmartTuna, parameterName=UseEventVO}
  n72 = TextureSampleParameter2D(UVs=@n60.UVs) {
   group=Visor Base
   parameterName=AnimatedDF_ColorTex
   samplerType=SAMPLERTYPE_Masks
   sortPriority=41
   texture=/Game/Global/Textures/Default/Blanks/AllRed
  }
  n73 = VectorParameter() {group=Visor Base, parameterName=AnimatedDF_Color, sortPriority=60, defaultValue=<rgba 0.7969 0 0 1>}
  NamedRerouteDeclaration(None=Multiply(
    A=Max(A=@n65, B=@n71)
    B=StaticSwitchParameter(True=@n72.RGB, False=@n73.RGB) {group=Visor Base, parameterName=AnimatedDF_UseColorTex, sortPriority=41}
   )) {name=AnimatedDF, nodeColor=<rgba 0.2188 1 0 1>}
  n74 = NamedRerouteUsage()
  n75 = StaticSwitchParameter(
   True=MaterialFunctionCall!(Main_TextureUV=@n74) {materialFunction=&/MF_GenericVisor_Glitch}.Result
   False=@n74
  ) {group=Visor Add-Ons, parameterName=Visor_UseGlitch, sortPriority=20}
  n76 = StaticSwitchParameter(True=Add(A=@n75, B=NamedRerouteUsage()), False=@n75) {group=Visor Add-Ons, parameterName=Visor_UseSweepDistortion, sortPriority=49}
  n77 = NamedRerouteUsage()
  NamedRerouteDeclaration(None=StaticSwitchParameter(True=Add(A=@n76, B=@n77), False=@n76) {group=Visor Add-Ons, parameterName=Visor_UseTVStatic, sortPriority=44}) {name=UVs_Distorted, nodeColor=<rgba 1 0 0.5391 1>}
  n78 = NamedRerouteUsage()
  n79 = ScalarParameter() {defaultValue=0.5, group=Visor Add-Ons, parameterName=Pulse_UVScale, sortPriority=55}
  n80 = Multiply(A=ScalarParameter() {defaultValue=16, group=Visor Add-Ons, parameterName=Pulse_Speed, sortPriority=55}, B=Time())
  n81 = TextureSampleParameter2D(UVs=Add(A=Multiply(A=@n78, B=@n79), B=Frac(None=@n80))) {
   group=Visor Add-Ons
   parameterName=Pulse_Gradient
   sortPriority=55
   texture=/Game/Global/Textures/Default/Masks_and_Gradients/T_Vert_Gradient_Sections
  }
  NamedRerouteDeclaration(None=LinearInterpolate(
    A=ScalarParameter() {defaultValue=0.7, group=Visor Add-Ons, parameterName=Pulse_BrightnessMin, sortPriority=55}
    B=ScalarParameter() {defaultValue=1.2, group=Visor Add-Ons, parameterName=Pulse_BrightnessMax, sortPriority=55}
    Alpha=@n81.R
   )) {name=Pulse, nodeColor=<rgba 0 1 0.625 1>}
  n82 = TextureCoordinate() {coordinateIndex=1}
  n83 = OneMinus(None=ComponentMask(None=Step(Y=@n82) {constX=0, constY=1}) {r=true})
  n84 = FunctionInput() {inputName=Tech Art Mask, sortPriority=1}
  n85 = NamedRerouteDeclaration(None=StaticSwitchParameter(True=Saturate(None=@n83), False=ChannelMaskParameter(None=@n84) {group=Visor, parameterName=Visor_Channel, sortPriority=19}.R) {group=Visor, parameterName=Visor_UseUV2forVisorMask, sortPriority=20}) {name=VisorMask, nodeColor=<rgba 1 0 0.5156 1>}
  NamedRerouteDeclaration(None=@n85) {name=TVStaticMask, nodeColor=<rgba 0.7109 1 0 1>}
  NamedRerouteDeclaration(None=@n85) {name=ScanlinesMask, nodeColor=<rgba 1 0.5116 0.5981 1>}
  NamedRerouteDeclaration(None=@n85) {name=SweepDistortionMask, nodeColor=<rgba 0.161 1 0.707 1>}
  NamedRerouteDeclaration() {name=Debug, nodeColor=<rgba 0 1 0.5078 1>}
  NamedRerouteUsage()
  NamedRerouteUsage()
  NamedRerouteUsage()
  NamedRerouteUsage()
  NamedRerouteUsage()
  NamedRerouteUsage()
  NamedRerouteUsage()
  StaticSwitchParameter(True=@n4, False=@n2.MaterialAttributes) {group=Visor, parameterName=UseGenericVisor, sortPriority=-1}
  TextureCoordinate()
  Time()
 ]}
